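# Soft daytime cave ambience cue for the Artemis Haunted House POI, on the indoor ambient sound class.
# Mixes a looping cave wind wave with a quieter cave ambience loop, each through an endlessly looping Enveloper whose roughly opposing pitch and volume curves add slow drift.
# The top quality level plays the mix, the next only the wind, and the other two are empty.
# Concurrency caps it at four instances, preventing new ones.
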SoundCue /Game/Sounds/Artemis/POI/HauntedHouse/Artemis_Cave_Soft_Day_Cue
root=QualityLevel_0
nodes [
 {
  id=Enveloper_1
  children [WavePlayer_1]
  properties {
   bLoop=true
   bLoopIndefinitely=true
   loopEnd=24
   pitchCurve {editorCurveData {keys [{time=0}, <at 5 1.1>, {time=10}, <at 14.9175 0.9>, <at 20 1.0001>]}}
   volumeCurve {editorCurveData {keys [<at 0 1>,<at 7.6503 0.7>,<at 17.1067 1.2>,<at 23.674 1.0001>]}}
  }
 }
 {
  id=Enveloper_12
  children [WavePlayer_21]
  properties {
   bLoop=true
   bLoopIndefinitely=true
   loopEnd=24
   pitchCurve {editorCurveData {keys [
      {time=0}
      <at 1.2275 0.9965>
      <at 2.3275 0.974>
      <at 3.6475 0.9595>
      <at 5 0.9>
      <at 8.5425 0.945>
      {time=10}
      <at 12.255 1.0585>
      <at 14.9175 1.1>
      <at 17.1225 1.0378>
      <at 18.965 1.0044>
      <at 20 1.0001>
     ]}}
   volumeCurve {editorCurveData {keys [<at 0 1>,<at 1.9087 1.005>,<at 4.3826 1.0657>,<at 7.6177 1.2>,<at 9.656 1.0727>,<at 11.7719 0.9847>,<at 13.074 0.9617>,<at 14.7992 0.9187>,<at 15.8409 0.7454>,<at 18.8031 0.9087>,<at 19.7471 0.9807>,<at 21.9931 0.9918>,<at 23.674 1.0001>]}}
  }
 }
 {id=Mixer_7, children [Enveloper_1,Enveloper_12], properties {inputVolume [0.9,0.25]}}
 {id=QualityLevel_0, children [Mixer_7,Enveloper_1,null,null]}
 {id=WavePlayer_1, properties {bLooping=true, soundWaveAssetPtr=/Game/Sounds/Fort_Ambient_Looping/Cave_Mixes/Cave_WindMid02}}
 {id=WavePlayer_21, properties {bLooping=true, soundWaveAssetPtr=/Game/Sounds/Fort_Ambient_Looping/Cave_Mixes/fort_cave_amb_01aLoop}}
]
properties {
 bOverrideConcurrency=true
 maxDistance=2097152
 soundClassObject=&Fort_Ambient_Indoors_Athena
 volumeMultiplier=0.752
 attenuationOverrides {attenuationShape=Box, distanceAlgorithm=Logarithmic, falloffDistance=1024, attenuationShapeExtents=<512 512 512>}
 concurrencyOverrides {maxCount=4, resolutionRule=PreventNew}
}
