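# Looping indoor ambience cue for the basement of the Haunted House point of interest.
# A Mixer root layers two looping wave players, Main_01 and Main_02, with the first weighted louder than the second.
# It uses the shared Athena ambient attenuation and routes through the Athena indoor ambient sound class.
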SoundCue /Game/Sounds/Artemis/POI/HauntedHouse/HauntedHouse_AmbientLoop_Basement_Cue
root=Mixer_0
nodes [
 {id=Mixer_0, children [WavePlayer_0,WavePlayer_1], properties {inputVolume [1.2,0.7]}}
 {
  id=WavePlayer_0
  properties {bLooping=true, soundWaveAssetPtr=&/HauntedHouse_AmbientLoop_Basement_Main_01}
 }
 {
  id=WavePlayer_1
  properties {bLooping=true, soundWaveAssetPtr=&/HauntedHouse_AmbientLoop_Basement_Main_02}
 }
]
properties {
 attenuationSettings=&Fort_Ambient_Athena_Att
 maxDistance=2097152
 soundClassObject=&Fort_Ambient_Indoors_Athena
 volumeMultiplier=1.002
 preEffectBusSends [{}]
}
