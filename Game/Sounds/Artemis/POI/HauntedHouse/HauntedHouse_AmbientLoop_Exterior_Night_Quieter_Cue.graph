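# Ambient sound cue for the Haunted House POI exterior at night, named as a quieter version, with a reduced overall volume.
# A Mixer layers four looping waves at separate levels: two main beds, wind, and a house creak mixed lowest.
# Only the creak passes through a Modulator, with equal pitch limits that fix it at a lowered pitch.
# Playback uses the Athena outdoor ambient sound class and shared Athena ambient attenuation.
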SoundCue /Game/Sounds/Artemis/POI/HauntedHouse/HauntedHouse_AmbientLoop_Exterior_Night_Quieter_Cue
root=Mixer_1
nodes [
 {id=Mixer_1, children [WavePlayer_2,WavePlayer_3,WavePlayer_6,Modulator_0], properties {inputVolume [0.8,1,0.6,0.2]}}
 {id=Modulator_0, children [WavePlayer_7], properties {pitchMax=0.8, pitchMin=0.8}}
 {
  id=WavePlayer_2
  properties {bLooping=true, soundWaveAssetPtr=&/HauntedHouse_AmbientLoop_Exterior_Main_01}
 }
 {
  id=WavePlayer_3
  properties {bLooping=true, soundWaveAssetPtr=&/HauntedHouse_AmbientLoop_Exterior_Main_02}
 }
 {
  id=WavePlayer_6
  properties {bLooping=true, soundWaveAssetPtr=&/HauntedHouse_AmbientLoop_Exterior_Wind_01}
 }
 {
  id=WavePlayer_7
  properties {bLooping=true, soundWaveAssetPtr=&/HauntedHouse_AmbientLoop_Exterior_HouseCreak_01}
 }
]
properties {
 attenuationSettings=&Fort_Ambient_Athena_Att
 maxDistance=2097152
 soundClassObject=&Fort_Ambient_Outdoors_Athena
 volumeMultiplier=0.251
 preEffectBusSends [{}]
}
